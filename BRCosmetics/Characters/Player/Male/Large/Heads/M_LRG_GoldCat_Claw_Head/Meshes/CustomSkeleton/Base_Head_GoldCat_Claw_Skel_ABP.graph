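# Animation Blueprint for the GoldCat Claw large male head cosmetic that adds physics-driven secondary motion to its ears and fur.
# The input pose passes through six chained single-bone Anim Dynamics nodes with angular limits and custom gravity; ears add damping and spring overrides.
# Each update sets their shared simulation space and alpha: World at reduced strength in the front end outside emotes and skydiving, otherwise Component at full alpha.
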AnimBlueprint /BRCosmetics/Characters/Player/Male/Large/Heads/M_LRG_GoldCat_Claw_Head/Meshes/CustomSkeleton/Base_Head_GoldCat_Claw_Skel_ABP
target_skeleton=&/Fortnite_Base_Head_GoldCat_Claw_Skeleton
graphs [
 AnimGraph {nodes [
   n1 = Get AnimDynamics_SimSpace_Frontend() -> (AnimDynamics_SimSpace_Frontend:AnimPhysSimSpaceType Enum)
   n2 = Get Anim_Dynamics_Alpha_Frontend() -> (Anim_Dynamics_Alpha_Frontend:double)
   n3 = Anim Dynamics - Bone: dyn_ear_l!(SimulationSpace:AnimPhysSimSpaceType Enum=@n1, ComponentPose=Local To Component(LocalPose=Input Pose()), Alpha=@n2, ExternalForce="0,0,0") -> (Pose) {
    angularBiasOverride=0.9999
    angularDampingOverride=0.9999
    angularSpringConstant=15000
    bGravityOverrideInSimSpace=true
    bOverrideAngularBias=true
    bOverrideAngularDamping=true
    bOverrideLinearDamping=true
    bUseGravityOverride=true
    linearDampingOverride=0.9999
    simulationSpace=Actor
    alphaBoolBlend {blendOption=Linear}
    alphaScaleBiasClamp {scale=0.52}
    boundBone {boneName=dyn_ear_l}
    componentAppliedLinearAccClamp=<1900 1900 1900>
    componentLinearAccScale=<0.47 0.47 0.47>
    componentLinearVelScale=<0.47 0.47 0.47>
    gravityOverride=<1 1 -200>
    physicsBodyDefinitions [{
      sphereCollisionRadius=0
      boundBone {boneName=dyn_ear_l}
      boxExtents=<5.5 5.5 1.8>
      constraintSetup {angularLimitsMax=<0 23 0>, angularLimitsMin=<0 -3 0>, angularTarget=<0 0 0>}
      localJointOffset=<1.9 0 0>
     }]
    relativeSpaceBone {boneName=spine_05}
   }
   n4 = Anim Dynamics - Bone: dyn_ear_r!(SimulationSpace:AnimPhysSimSpaceType Enum=@n1, ComponentPose=@n3, Alpha=@n2, ExternalForce="0,0,0") -> (Pose) {
    angularBiasOverride=0.9999
    angularDampingOverride=0.9999
    angularSpringConstant=15000
    bGravityOverrideInSimSpace=true
    bOverrideAngularBias=true
    bOverrideAngularDamping=true
    bOverrideLinearDamping=true
    bUseGravityOverride=true
    linearDampingOverride=0.9999
    simulationSpace=Actor
    alphaBoolBlend {blendOption=Linear}
    alphaScaleBiasClamp {scale=0.7}
    boundBone {boneName=dyn_ear_r}
    componentAppliedLinearAccClamp=<2100 2100 2100>
    componentLinearAccScale=<0.51 0.51 0.51>
    componentLinearVelScale=<0.51 0.51 0.51>
    gravityOverride=<1 1 -200>
    physicsBodyDefinitions [{
      sphereCollisionRadius=0
      boundBone {boneName=dyn_ear_r}
      boxExtents=<6.3 6.3 2.2>
      constraintSetup {angularLimitsMax=<0 29 0>, angularLimitsMin=<0 1 0>, angularTarget=<0 0 0>}
      localJointOffset=<2.1 0 0>
     }]
    relativeSpaceBone {boneName=pelvis}
   }
   n5 = Anim Dynamics - Bone: dyn_fur_bot_l!(SimulationSpace:AnimPhysSimSpaceType Enum=@n1, ComponentPose=@n4, Alpha=@n2, ExternalForce="0,0,0") -> (Pose) {
    bGravityOverrideInSimSpace=true
    bUseGravityOverride=true
    simulationSpace=Actor
    alphaBoolBlend {blendOption=Linear}
    alphaScaleBiasClamp {scale=0.9}
    boundBone {boneName=dyn_fur_bot_l}
    componentAppliedLinearAccClamp=<1900 1900 1900>
    componentLinearAccScale=<0.45 0.45 0.45>
    componentLinearVelScale=<0.45 0.45 0.45>
    gravityOverride {z=-98}
    physicsBodyDefinitions [{
      sphereCollisionRadius=0
      boundBone {boneName=dyn_fur_bot_l}
      boxExtents=<6 6 2>
      constraintSetup {angularLimitsMax=<0 12 6>, angularLimitsMin=<0 -8 -4>, angularTarget=<0 0 0>}
      localJointOffset=<2 0 0>
     }]
    relativeSpaceBone {boneName=spine_05}
   }
   n6 = Anim Dynamics - Bone: dyn_fur_bot_r!(SimulationSpace:AnimPhysSimSpaceType Enum=@n1, ComponentPose=@n5, Alpha=@n2, ExternalForce="0,0,0") -> (Pose) {
    bGravityOverrideInSimSpace=true
    bUseGravityOverride=true
    simulationSpace=Actor
    alphaBoolBlend {blendOption=Linear}
    alphaScaleBiasClamp {scale=1.05}
    boundBone {boneName=dyn_fur_bot_r}
    componentAppliedLinearAccClamp=<2050 2050 2050>
    componentLinearAccScale=<0.5 0.5 0.5>
    componentLinearVelScale=<0.5 0.5 0.5>
    gravityOverride {z=-98}
    physicsBodyDefinitions [{
      sphereCollisionRadius=0
      boundBone {boneName=dyn_fur_bot_r}
      boxExtents=<6 6 2>
      constraintSetup {angularLimitsMax=<0 8 4>, angularLimitsMin=<0 -12 -6>, angularTarget=<0 0 0>}
      localJointOffset=<2 0 0>
     }]
    relativeSpaceBone {boneName=spine_05}
   }
   n7 = Anim Dynamics - Bone: dyn_fur_top_l!(SimulationSpace:AnimPhysSimSpaceType Enum=@n1, ComponentPose=@n6, Alpha=@n2, ExternalForce="0,0,0") -> (Pose) {
    bGravityOverrideInSimSpace=true
    bUseGravityOverride=true
    simulationSpace=Actor
    alphaBoolBlend {blendOption=Linear}
    alphaScaleBiasClamp {scale=0.95}
    boundBone {boneName=dyn_fur_top_l}
    componentAppliedLinearAccClamp=<2100 2100 2100>
    componentLinearAccScale=<0.55 0.55 0.55>
    componentLinearVelScale=<0.55 0.55 0.55>
    gravityOverride {z=-98}
    physicsBodyDefinitions [{
      sphereCollisionRadius=0
      boundBone {boneName=dyn_fur_top_l}
      boxExtents=<6 6 2>
      constraintSetup {angularLimitsMax=<0 8 4>, angularLimitsMin=<0 -9 -4>, angularTarget=<0 0 0>}
      localJointOffset=<2 0 0>
     }]
    relativeSpaceBone {boneName=spine_05}
   }
   n8 = Anim Dynamics - Bone: dyn_fur_top_r!(SimulationSpace:AnimPhysSimSpaceType Enum=@n1, ComponentPose=@n7, Alpha=@n2, ExternalForce="0,0,0") -> (Pose) {
    bGravityOverrideInSimSpace=true
    bUseGravityOverride=true
    simulationSpace=Actor
    alphaBoolBlend {blendOption=Linear}
    boundBone {boneName=dyn_fur_top_r}
    componentAppliedLinearAccClamp=<2000 2000 2000>
    componentLinearAccScale=<0.5 0.5 0.5>
    componentLinearVelScale=<0.5 0.5 0.5>
    gravityOverride {z=-98}
    physicsBodyDefinitions [{
      sphereCollisionRadius=0
      boundBone {boneName=dyn_fur_top_r}
      boxExtents=<6 6 2>
      constraintSetup {angularLimitsMax=<0 9 5>, angularLimitsMin=<0 -8 -4>, angularTarget=<0 0 0>}
      localJointOffset=<2 0 0>
     }]
    relativeSpaceBone {boneName=spine_05}
   }
   Output Pose(Result=Component To Local(ComponentPose=@n8))
  ]}
 EventGraph {nodes [
   Event BlueprintUpdateAnimation()
   n1 = NOT Boolean(A=Get bIsPlayingEmote())
   n2 = NOT Boolean(A=Get bIsSkyDiving())
   > n3 = Branch(Condition=AND Boolean(A=Get bIsInFrontEnd(), B=AND Boolean(A=@n2, B=@n1)))
   > Set AnimDynamics_SimSpace_Frontend(AnimDynamics_SimSpace_Frontend:AnimPhysSimSpaceType Enum=World) -> (Output_Get:AnimPhysSimSpaceType Enum)
   > Set Anim_Dynamics_Alpha_Frontend(Anim_Dynamics_Alpha_Frontend:double=0.4) -> (Output_Get:double)
   Set AnimDynamics_SimSpace_Frontend(execute=@n3.else, AnimDynamics_SimSpace_Frontend=Component)
   > Set Anim_Dynamics_Alpha_Frontend(Anim_Dynamics_Alpha_Frontend=1)
   TryGetPawnOwner()
  ]}
]
variables [
 Anim_Dynamics_Alpha_Frontend {type=number}
 AnimDynamics_SimSpace_Frontend {type=AnimPhysSimSpaceType}
]
